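# Material function in the ComplexMotion library that samples one texture four times, drifting along four opposing diagonals to produce chaotic motion.
# Time scaled by Speed drives four Panners, three with fixed UV offsets to break up alignment.
# The summed RGB is multiplied by Divisor, default a quarter.
# Quality switches zero out samples, leaving two at Medium and one at Low, without adjusting that multiplier.
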MaterialFunction /Engine/Functions/Engine_MaterialFunctions03/ComplexMotion/Motion_4WayChaos
graph {
 properties {description=4-way diagonal motion, libraryCategoriesText [Misc, ComplexMotion]}
 nodes [
  n1 = FunctionInput(Preview=TextureCoordinate()) {bUsePreviewValueAsDefault=true, inputName=Coordinates, inputType=FunctionInput_Vector2}
  n2 = FunctionInput() {bUsePreviewValueAsDefault=true, inputName=Speed, inputType=FunctionInput_Scalar, previewValue=<1 0 0 1>}
  n3 = Multiply(A=@n2, B=Time())
  n4 = Panner(Coordinate=@n1, Time=@n3) {speedX=0.1, speedY=0.1}
  n5 = FunctionInput(Preview=TextureObject()) {inputName=Texture, inputType=FunctionInput_Texture2D, sortPriority=1}
  n6 = Add(A=@n1, B=Constant2Vector() {g=0.3548, r=0.4181})
  n7 = Add(
   A=QualitySwitch(Default=TextureSample(UVs=@n4, Tex=@n5).RGB, Low=Constant())
   B=TextureSample(UVs=Panner(Coordinate=@n6, Time=@n3) {speedX=-0.1, speedY=-0.1}, Tex=@n5).RGB
  )
  n8 = Panner(Coordinate=Add(A=@n1, B=Constant2Vector() {g=0.1484, r=0.8649}), Time=@n3) {speedX=-0.1, speedY=0.1}
  n9 = Panner(Coordinate=Add(A=@n1, B=Constant2Vector() {g=0.7516, r=0.6513}), Time=@n3) {speedX=0.1, speedY=-0.1}
  n10 = Constant()
  n11 = QualitySwitch(Default=Add(A=TextureSample(UVs=@n8, Tex=@n5).RGB, B=TextureSample(UVs=@n9, Tex=@n5).RGB), Low=@n10, Medium=@n10)
  n12 = FunctionInput() {
   bUsePreviewValueAsDefault=true
   description=How much to reduce the overall combined results of the 4 way add.
   inputName=Divisor
   inputType=FunctionInput_Scalar
   previewValue=<0.25 0 0 1>
  }
  FunctionOutput(None=Multiply(A=Add(A=@n7, B=@n11), B=@n12))
 ]
}
